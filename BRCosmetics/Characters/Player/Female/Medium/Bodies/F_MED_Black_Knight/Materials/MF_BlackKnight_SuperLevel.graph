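# Material function that adds the female Black Knight skin's super-level look, spreading a smoky red metallic, emissive finish over the input material attributes.
# Blueprint-fed game time and current level drive an animated, optionally quantized progress between start and stop levels.
# Progress thresholds a noise-distorted gradient texture into the transition mask, with a sphere mask adding edge glow.
# Noise is triplanar-projected from pre-skinned position and normal; several branches, including the curve-atlas lookup, are unconnected.
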MaterialFunction /BRCosmetics/Characters/Player/Female/Medium/Bodies/F_MED_Black_Knight/Materials/MF_BlackKnight_SuperLevel
graph {nodes [
  n1 = ScalarParameter() {defaultValue=100, group=SuperLevelling_Mask, parameterName=BKSL_StartAtLevel, sortPriority=3}
  n2 = Time()
  n3 = ScalarParameter() {defaultValue=-99999, group=SuperLevelling_Mask, parameterName=BKSL_GameTimeFromBP, sortPriority=100}
  n4 = ScalarParameter() {defaultValue=2, group=SuperLevelling_Mask, parameterName=BKSL_Animation_Duration, sortPriority=101}
  n5 = Saturate(None=Divide(A=Subtract(A=@n2, B=@n3), B=Add(A=@n4) {constB=0}))
  n6 = LinearInterpolate(A=@n1, B=ScalarParameter() {group=SuperLevelling_Mask, parameterName=BKSL_CurrentLevel, sortPriority=3}, Alpha=@n5)
  n7 = ScalarParameter() {defaultValue=160, group=SuperLevelling_Mask, parameterName=BKSL_StopAtLevel, sortPriority=3}
  n8 = Divide(A=Subtract(A=Clamp(None=@n6, Min=@n1, Max=@n7), B=@n1), B=Subtract(A=@n7, B=@n1))
  n9 = Max(A=Ceil(None=ScalarParameter() {defaultValue=1, group=SuperLevelling_Mask, parameterName=BKSL_QuantizationSteps, sliderMax=100, sliderMin=1}))
  n10 = Multiply(A=@n8, B=@n9)
  n11 = Subtract(A=@n9)
  n12 = LinearInterpolate(A=@n8, B=Divide(A=Floor(None=@n10), B=@n9), Alpha=Saturate(None=Ceil(None=@n11)))
  n13 = CurveAtlasRowParameter(CurveTime=@n12) {
   atlas=&AthenaCurveAtlas01
   curve=/Game/Athena/LinearColorCurves/Curve_SuperlevelMask
   defaultValue=205
   group=SuperLevelling_Mask
   parameterName=BKSLSL_SuperLevelMaskCurve
  }
  ChannelMaskParameter(None=@n13) {group=SuperLevelling_Mask, maskChannel=Green, parameterName=BKSL_SuperLevelCurveChannel, defaultValue=<rgba 0 1 0 0>}
  n14 = ScalarParameter() {group=SuperLevelling_Mask, parameterName=BKSL_CurrentLevel, sortPriority=3}
  n15 = LinearInterpolate(A=Subtract(A=@n14), B=@n14, Alpha=@n5)
  n16 = ScalarParameter() {defaultValue=100, group=SuperLevelling_Mask, parameterName=BKSL_StartAtLevel, sortPriority=3}
  n17 = ScalarParameter() {defaultValue=160, group=SuperLevelling_Mask, parameterName=BKSL_StopAtLevel, sortPriority=3}
  n18 = Divide(A=Subtract(A=Clamp(None=@n15, Min=@n16, Max=@n17), B=@n16), B=Subtract(A=@n17, B=@n16))
  n19 = Max(A=Ceil(None=ScalarParameter() {defaultValue=1, group=SuperLevelling_Mask, parameterName=BKSL_QuantizationSteps, sliderMax=100, sliderMin=1}))
  n20 = Multiply(A=@n18, B=@n19)
  n21 = Subtract(A=@n19)
  n22 = LinearInterpolate(A=@n18, B=Divide(A=Floor(None=@n20), B=@n19), Alpha=Saturate(None=Ceil(None=@n21)))
  Clamp(None=@n22)
  n23 = FunctionInput(Preview=VertexInterpolator(VS=PreSkinnedPosition()).PS) {bUsePreviewValueAsDefault=true, inputName=Local Position}
  n24 = VectorParameter() {group=SuperLevelling_Mask, parameterName=BKSL_Noise_01_PanSpeedAndDirection, defaultValue=<rgba 0 0 -0.1 1>}
  n25 = Add(
   A=Divide(A=@n23, B=VectorParameter() {group=SuperLevelling_Mask, parameterName=BKSL_Noise_01_Tiling, defaultValue=<rgba 48 48 48 1>}.RGB)
   B=Multiply(A=Time(), B=@n24.RGB)
  )
  n26 = ComponentMask(None=@n25) {b=true, g=true}
  n27 = ComponentMask(None=@n25) {b=true, r=true}
  n28 = Abs(None=VertexInterpolator(VS=PreSkinnedNormal()).PS)
  n29 = ComponentMask(None=@n28) {g=true}
  n30 = ComponentMask(None=@n28) {b=true}
  n31 = LinearInterpolate(
   A=LinearInterpolate(
    A=TextureSampleParameter2D(UVs=@n26) {group=SuperLevelling_Mask, parameterName=BKSL_Noise_01, texture=&T_NoisePack_01}.RGB
    B=TextureSampleParameter2D(UVs=@n27) {group=SuperLevelling_Mask, parameterName=BKSL_Noise_01, texture=&T_NoisePack_01}.RGB
    Alpha=@n29
   )
   B=TextureSampleParameter2D(UVs=ComponentMask(None=@n25) {g=true, r=true}) {group=SuperLevelling_Mask, parameterName=BKSL_Noise_01, texture=&T_NoisePack_01}.RGB
   Alpha=@n30
  )
  n32 = Subtract(A=Multiply(A=@n31) {constB=2})
  n33 = ComponentMask(None=@n32) {g=true, r=true}
  n34 = ScalarParameter() {defaultValue=0.01, group=SuperLevelling_Mask, parameterName=BKSL_GradientDistortioNSterngth}
  n35 = TextureSampleParameter2D(UVs=Add(A=TextureCoordinate(), B=Multiply(A=@n33, B=@n34))) {
   group=SuperLevelling_Mask
   parameterName=BKSL_FlowMap
   samplerType=SAMPLERTYPE_LinearColor
   texture=/BRCosmetics/Characters/Player/Female/Medium/Bodies/F_MED_Black_Knight/Textures/TransitionGradient
  }
  ComponentMask(None=@n35.RGBA) {g=true, r=true}
  Frac(None=Multiply(A=Time(), B=ScalarParameter() {defaultValue=0.05, parameterName=BKSL_DebugSpeeeed}))
  n36 = GetMaterialAttributes(None=FunctionInput() {inputName=Material, inputType=FunctionInput_MaterialAttributes}) {attributeGetTypes [
    #1
    #2
    #3
    #4
   ]}
  n37 = VectorParameter() {group=SuperLevelling_Mask, parameterName=BKSL_BaseColor, defaultValue=<rgba 0.6823 0 0 1>}
  n38 = CameraVectorWS()
  n39 = Transform(None=@n36.Normal)
  n40 = OneMinus(None=Clamp(None=DotProduct(A=@n38, B=@n39)))
  n41 = Power(Base=@n40)
  n42 = Desaturation(None=@n36.BaseColor, Fraction=Constant() {r=0.75})
  n43 = ComponentMask(None=@n31) {r=true}
  n44 = LinearInterpolate(
   A=ScalarParameter() {defaultValue=0.25, group=SuperLevelling_Mask, parameterName=BKSL_Darkness_Min}
   B=ScalarParameter() {defaultValue=0.5, group=SuperLevelling_Mask, parameterName=BKSL_Darkness_Max}
   Alpha=@n43
  ) {constA=0.25, constB=0.5}
  n45 = TextureSampleParameter2D(UVs=TextureCoordinate()) {
   group=SuperLevelling_Mask
   parameterName=BKSL_SuperLevelMasks_02
   texture=/BRCosmetics/Characters/Player/Female/Medium/Bodies/F_MED_Black_Knight/Textures/T_BlackKnight_Body_SL_Mask
  }
  n46 = Clamp(None=@n45.R)
  n47 = Add(A=@n46, B=@n40)
  n48 = OneMinus(None=@n43)
  n49 = LinearInterpolate(
   A=@n37.RGB
   B=VectorParameter() {group=SuperLevelling_Mask, parameterName=BKSL_BaseColor_02, defaultValue=<rgba 0.0677 0.0068 0.0068 1>}.RGB
   Alpha=Clamp(None=Multiply(A=@n47, B=@n48))
  )
  n50 = Constant() {r=1}
  n51 = Multiply(
   A=FunctionInput(Preview=@n50) {bUsePreviewValueAsDefault=true, inputName=AO, inputType=FunctionInput_Scalar}
   B=FunctionInput(Preview=@n50) {bUsePreviewValueAsDefault=true, inputName=GMap, inputType=FunctionInput_Scalar}
  )
  n52 = ComponentMask(None=@n31) {g=true}
  n53 = Subtract(
   A=Multiply(A=Add(A=@n46, B=@n51))
   B=Multiply(A=@n52, B=ScalarParameter() {defaultValue=1.5, group=SuperLevelling_Mask, parameterName=BKSL_NoiseStrength})
  )
  n54 = Multiply(
   A=Multiply(A=@n46, B=Clamp(None=@n53))
   B=Multiply(A=Clamp(None=@n52) {minDefault=0.5}, B=ScalarParameter() {defaultValue=2, group=SuperLevelling_Mask, parameterName=BKSL_SmokeStrength})
  )
  n55 = LinearInterpolate(A=Add(A=Multiply(A=@n37.RGB, B=@n41), B=Multiply(A=@n42, B=@n44) {constB=0.25}), B=@n49, Alpha=@n54)
  n56 = Add(A=@n35.R, B=ScalarParameter() {defaultValue=0.61, group=SuperLevelling_Mask, parameterName=BKSL_TransitionScalarBias_01}) {constB=0}
  n57 = ScalarParameter() {defaultValue=1, parameterName=BKSL_TransitionScalarMult}
  n58 = Subtract(A=@n56, B=Multiply(A=OneMinus(None=@n22), B=@n57) {constB=2})
  n59 = VectorParameter() {group=SuperLevelling_Mask, parameterName=BKSL_TransitionNoise_Tiling, defaultValue=<rgba 24 24 24 1>}
  n60 = Time()
  n61 = VectorParameter() {group=SuperLevelling_Mask, parameterName=BKSL_TransitionNoise_PanSpeedAndDirection, defaultValue=<rgba 0 0 0 1>}
  n62 = Add(
   A=Multiply(A=@n32, B=ScalarParameter() {defaultValue=0.1, group=SuperLevelling_Mask, parameterName=BKSL_NoiseDistortio_Sterngth})
   B=Add(A=Divide(A=@n23, B=@n59.RGB), B=Multiply(A=@n60, B=@n61.RGB))
  )
  n63 = TextureSampleParameter2D(UVs=ComponentMask(None=@n62) {b=true, g=true}) {
   group=SuperLevelling_Mask
   parameterName=BKSL_TransitionNoise
   texture=&T_BlackKnight_Body_SL_Mask
  }
  n64 = TextureSampleParameter2D(UVs=ComponentMask(None=@n62) {b=true, r=true}) {
   group=SuperLevelling_Mask
   parameterName=BKSL_TransitionNoise
   texture=&T_BlackKnight_Body_SL_Mask
  }
  n65 = TextureSampleParameter2D(UVs=ComponentMask(None=@n62) {g=true, r=true}) {
   group=SuperLevelling_Mask
   parameterName=BKSL_TransitionNoise
   texture=&T_BlackKnight_Body_SL_Mask
  }
  n66 = LinearInterpolate(A=LinearInterpolate(A=@n63.RGB, B=@n64.RGB, Alpha=@n29), B=@n65.RGB, Alpha=@n30)
  n67 = Multiply(
   A=OneMinus(None=ComponentMask(None=@n66) {g=true})
   B=ScalarParameter() {defaultValue=0.1, group=SuperLevelling_Mask, parameterName=SL_Transition_Falloff_Start}
  )
  n68 = Clamp(None=Round(None=Subtract(A=@n58, B=@n67)))
  n69 = Multiply(A=@n49, B=@n54)
  n70 = Add(A=Multiply(A=Multiply(A=@n69, B=@n68), B=ScalarParameter() {defaultValue=1, parameterName=BKSL_EmissveStrength}), B=@n36.EmissiveColor)
  n71 = VectorParameter() {group=SuperLevelling_Mask, parameterName=BKSL_EdgeGlowColor, defaultValue=<rgba 0.6823 0.0995 0 1>}
  n72 = Multiply(A=OneMinus(None=@n12), B=@n57) {constB=2}
  n73 = SphereMask(
   A=Subtract(A=Subtract(A=@n56, B=@n72), B=@n67)
   B=ScalarParameter() {defaultValue=0.5, group=SuperLevelling_Mask, parameterName=BKSL_EdgeMask_SamplePosition}
   Radius=ScalarParameter() {defaultValue=0.01, group=SuperLevelling_Mask, parameterName=BKSL_EdgeMask_Radius}
   Hardness=ScalarParameter() {group=SuperLevelling_Mask, parameterName=BKSL_EdgeMask_Hardness}
  )
  n74 = Power(Base=Multiply(A=Multiply(A=@n71.RGB, B=@n73), B=ScalarParameter() {defaultValue=5, group=SuperLevelling_Mask, parameterName=BKSL_EdgeGlow_Strength}))
  n75 = LinearInterpolate(A=@n36.Metallic, B=Constant() {r=1}, Alpha=@n54)
  FunctionOutput(None=SetMaterialAttributes!(
    MaterialAttributes=@n36.MaterialAttributes
    BaseColor=LinearInterpolate(A=@n36.BaseColor, B=@n55, Alpha=@n68)
    Metallic=LinearInterpolate(A=@n36.Metallic, B=@n75, Alpha=@n68)
    EmissiveColor=Add(A=@n70, B=@n74)
   ) {attributeSetTypes [#1, #3, #4]})
  LinearInterpolate(A=@n46, B=OneMinus(None=@n46), Alpha=ScalarParameter() {group=SuperLevelling_Mask, parameterName=BKSL_FlipMask})
  n76 = ScalarParameter() {defaultValue=1.33, group=SuperLevelling_Mask, parameterName=BKSL_Animation_EaseIn, sortPriority=103}
  n77 = Multiply(A=Subtract(A=Power(Base=@n5, Exp=@n76) {constExponent=1.33}) {constB=0.5}) {constB=2}
  Power(
   Base=OneMinus(None=Abs(None=@n77))
   Exp=ScalarParameter() {defaultValue=6, group=SuperLevelling_Mask, parameterName=BKSL_Animation_EaseInOut, sortPriority=103}
  ) {constExponent=8}
  Round(None=Clamp(None=@n22))
  Subtract(A=@n17, B=ScalarParameter() {defaultValue=1, parameterName=BKSL_StopAtLevelBias})
  VertexInterpolator(VS=PreSkinnedPosition())
 ]}
