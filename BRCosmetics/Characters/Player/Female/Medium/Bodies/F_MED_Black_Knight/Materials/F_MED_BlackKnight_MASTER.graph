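# Master material for the female medium Black Knight player body, using material attributes with per-expression shading models (Default Lit, Subsurface Profile, Cloth).
# A shared character PreFX function feeds a chain of static-switch layers: cloth fuzz, silk, metal LUT, thin film and generic gem, all masked by SkinFX_Mask.
# An optional Black Knight "SuperLevel" effect uses interpolated pre-skinned position.
# The shared PostFX function finishes the chain.
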Material /BRCosmetics/Characters/Player/Female/Medium/Bodies/F_MED_Black_Knight/Materials/F_MED_BlackKnight_MASTER
graph {
 properties {
  bUseMaterialAttributes=true
  bUsedWithMeshDeformer=true
  bUsedWithMorphTargets=true
  bUsedWithSkeletalMesh=true
  pixelDepthOffsetMode=PDOM_Legacy
  shadingModel=MSM_FromMaterialExpression
  usedShadingModels=Default Lit | Subsurface Profile | Cloth
 }
 nodes [
  n1 = MaterialFunctionCall() {materialFunction=&BaseFN_CharacterFunction_PreFX}
  n2 = VertexInterpolator(VS=PreSkinnedPosition())
  n3 = TextureSampleParameter2D() {parameterName=SkinFX_Mask}
  n4 = StaticSwitchParameter(
   True=MaterialFunctionCall!(
    Alpha=StaticComponentMaskParameter(None=@n3.RGBA) {defaultR=true, group=Skin FX, parameterName=Cloth Channel}
    Gmap=@n1.Gmap
    Material=@n1.Gmap
   ) {materialFunction=/Game/Weapons/WeaponSkins/MaterialFunction/ClothFuzz}.Result
   False=@n1.Result
  ) {group=Skin FX, parameterName=Use Cloth fuzz}
  n5 = StaticSwitchParameter(
   True=MaterialFunctionCall!(Alpha=ChannelMaskParameter(None=@n3.RGBA) {group=Skin FX, parameterName=Silk Channel}.R, Material=@n4) {materialFunction=/Game/Weapons/WeaponSkins/MaterialFunction/MF_Silk}.Result
   False=@n4
  ) {group=Skin FX, parameterName=Use Silk}
  n6 = StaticSwitchParameter(
   True=MaterialFunctionCall(Mask=@n3.RGBA, Materials=@n5) {materialFunction=&MF_MetalsLUT}.Result
   False=@n5
  ) {group=Skin FX, parameterName=UseMetalLUT}
  n7 = StaticSwitchParameter(
   True=MaterialFunctionCall!(Alpha=StaticComponentMaskParameter(None=@n3.RGBA) {defaultR=true, group=Skin FX, parameterName=ThinFilm_Channel}, Material Attributes=@n6) {materialFunction=/Game/Weapons/WeaponSkins/MaterialFunction/ThinFilmFunction}.Result
   False=@n6
  ) {group=Skin FX, parameterName=Use Thin Film}
  n8 = StaticSwitchParameter(
   True=MaterialFunctionCall!(
    MASK=StaticComponentMaskParameter(None=@n3.RGBA) {defaultR=true, group=Skin FX, parameterName=Generic Gem Effect Channel}
    Color Change Sequence
    Material=@n7
   ) {materialFunction=/Game/Weapons/WeaponSkins/MaterialFunction/GenericGem}.Result
   False=@n7
  ) {group=Skin FX, parameterName=Use Generic Gem}
  MaterialFunctionCall(
   AO=@n1.AO
   Distance Blend=@n1.AO
   Fresnel=@n1.AO
   Gmap=@n1.AO
   Material=StaticSwitchParameter(
    True=MaterialFunctionCall!(Local Position=@n2.PS, AO=@n1.AO, GMap=@n1.AO, Material=@n8) {materialFunction=&/MF_BlackKnight_SuperLevel}.Result
    False=@n8
   ) {group=Skin FX, parameterName=UseBKSuperLevel}
  ) {materialFunction=&BaseFN_CharacterFunction_PostFX}
  StaticComponentMaskParameter() {defaultR=true, group=Skin FX, parameterName=ThinFilm_Channel}
  TextureObjectParameter() {group=Skin FX, parameterName=SkinFX_Mask}
 ]
}
