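# Physics asset for the Black Knight female character's bandana, an armband whose knot and loose tails have physics bodies for secondary motion.
# A small sphere on the master knot bone anchors a knot capsule, from which high and low tails branch as two-capsule chains.
# Constraints link each body to its parent with moderate swing limits, wider at the knot, and very tight twist limits, so the cloth bends without spinning.
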PhysicsAsset /BRCosmetics/Characters/Player/Female/Medium/Bodies/F_MED_Black_Knight/Meshes/F_MED_Black_Knight_Bandana_Physics
bodies {
 dyn_armband_hi_01 [{shapeType=Capsule, length=1, radius=1.5, center=<1 0 0>, rotation=<rot 90 0 0>}]
 dyn_armband_hi_02 [{shapeType=Capsule, length=1, radius=1, center=<1.5 0 0>, rotation=<rot 90 0 0>}]
 dyn_armband_knot [{shapeType=Capsule, length=0.75, radius=2.5, center=<0.2015 0 0>, rotation=<rot 90 0 0>}]
 dyn_armband_knot_MSTR [{shapeType=Sphere, radius=0.505}]
 dyn_armband_low_01 [{shapeType=Capsule, length=1, radius=1.75, center=<1.2683 0 -0.0001>, rotation=<rot 90 0 0>}]
 dyn_armband_low_02 [{shapeType=Capsule, length=1, radius=1.5, center=<1.5712 0 -0.0001>, rotation=<rot 90 0 0>}]
}
constraints [
 {bone1Name=dyn_armband_knot, bone2Name=dyn_armband_knot_MSTR, swing1LimitDegrees=45, swing2LimitDegrees=45, twistLimitDegrees=2}
 {bone1Name=dyn_armband_low_01, bone2Name=dyn_armband_knot, swing1LimitDegrees=35, swing2LimitDegrees=35, twistLimitDegrees=2}
 {bone1Name=dyn_armband_low_02, bone2Name=dyn_armband_low_01, swing1LimitDegrees=35, swing2LimitDegrees=35, twistLimitDegrees=2}
 {bone1Name=dyn_armband_hi_01, bone2Name=dyn_armband_knot, swing1LimitDegrees=35, swing2LimitDegrees=35, twistLimitDegrees=2}
 {bone1Name=dyn_armband_hi_02, bone2Name=dyn_armband_hi_01, swing1LimitDegrees=35, swing2LimitDegrees=35, twistLimitDegrees=2}
]
