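# Physics asset for the knife and pouch on the F_MED_Black_Knight female player body cosmetic, defining capsule bodies for their secondary motion.
# The knife and pouch bodies are each constrained to a tiny driver capsule with narrow swing and twist limits, tighter in swing for the pouch.
# Larger capsules on both thighs, differing in length and offset, provide leg volumes alongside the dangling pieces.
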PhysicsAsset /BRCosmetics/Characters/Player/Female/Medium/Bodies/F_MED_Black_Knight/Meshes/F_MED_Black_Knight_Knife_Physics
bodies {
 dyn_driver_knife [{shapeType=Capsule, length=0.505, radius=0.505, rotation=<rot 0 -90 0>}]
 dyn_driver_pouch [{shapeType=Capsule, length=0.505, radius=0.505, rotation=<rot 0 -90 0>}]
 dyn_knife [{shapeType=Capsule, length=22.9132, radius=5.3783, center=<3.9408 -0.0252 5.1913>, rotation=<rot 0 89.9999 -93.1275>}]
 dyn_pouch [{shapeType=Capsule, length=2, radius=5.25, center=<8.4224 -0.0642 -7.7081>, rotation=<rot 75.2445 -0.0001 90.9237>}]
 thigh_l [{shapeType=Capsule, length=34.5379, radius=12.2527, center=<-8.2248 -0.7087 5.4354>, rotation=<rot -0.3957 -90.966 -91.7315>}]
 thigh_r [{shapeType=Capsule, length=19.2879, radius=12.2527, center=<3.444 -0.1596 -5.8188>, rotation=<rot -0.293 -87.5862 -103.4853>}]
}
constraints [
 {bone1Name=dyn_knife, bone2Name=dyn_driver_knife, swing1LimitDegrees=30, swing2LimitDegrees=30, twistLimitDegrees=4}
 {bone1Name=dyn_pouch, bone2Name=dyn_driver_pouch, swing1LimitDegrees=20, swing2LimitDegrees=15, twistLimitDegrees=8}
]
