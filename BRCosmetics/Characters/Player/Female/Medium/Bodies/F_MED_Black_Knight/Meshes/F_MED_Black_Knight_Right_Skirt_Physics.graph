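# Physics asset for the right skirt of the female medium Black Knight body, covering a dynamic shirt bone chain and the right leg.
# Capsule bodies link from thigh_r through parent and driver bones into four shirt segments, with calf_r joined to the thigh by an angularly free constraint.
# From the driver down, swing limits widen toward the tip while twist limits stay tight.
# The calf and upper three shirt segments each use paired capsules.
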PhysicsAsset /BRCosmetics/Characters/Player/Female/Medium/Bodies/F_MED_Black_Knight/Meshes/F_MED_Black_Knight_Right_Skirt_Physics
bodies {
 calf_r [
  {shapeType=Capsule, length=32.6595, radius=6.1206, center=<23.4703 2.5909 0.6593>, rotation=<rot 87.894 -180 -180>}
  {shapeType=Capsule, length=11.4095, radius=6.3706, center=<13.593 4.2679 -0.0866>, rotation=<rot 87.8439 179.9999 166.5306>}
 ]
 dyn_driver_shirt_r [{shapeType=Capsule, length=5.25, radius=6.755, center=<0.5 0 3>, rotation=<rot 90 0 0>}]
 dyn_parent_shirt_r [{shapeType=Capsule, length=0.505, radius=0.505, rotation=<rot 0 -90 0>}]
 dyn_shirt_01_r [
  {shapeType=Capsule, length=10.25, radius=4, center=<1.3516 -0.2039 4.1816>, rotation=<rot 88.0079 0 0>}
  {shapeType=Capsule, length=10.25, radius=4, center=<1.0752 6.9543 0.5059>, rotation=<rot 89.698 179.9999 179.9999>}
 ]
 dyn_shirt_02_r [
  {shapeType=Capsule, length=7, radius=4.255, center=<1.8685 -0.3751 3.2833>, rotation=<rot 86.1664 -179.9998 -179.9998>}
  {shapeType=Capsule, length=7, radius=4.255, center=<2.1007 6.5694 -0.4321>, rotation=<rot 86.0344 177.3675 177.6519>}
 ]
 dyn_shirt_03_r [
  {shapeType=Capsule, length=4, radius=3.5, center=<2.763 0.0001 2.4778>, rotation=<rot 89.228 -179.9996 -179.9996>}
  {shapeType=Capsule, length=4, radius=3.255, center=<2.5995 4.0433 0.7018>, rotation=<rot 85.2807 -179.9999 -179.2227>}
 ]
 dyn_shirt_04_r [{
   shapeType=Capsule
   length=7.6711
   radius=1.25
   shapeName=dyn_shirt_04_r_capsule
   center=<3.3368 0.9711 0.2692>
   rotation=<rot 0 98.3969 -87.0328>
  }]
 thigh_r [{shapeType=Capsule, length=29.7023, radius=10.9286, center=<22.7706 -1.1794 -1.7722>, rotation=<rot -0.0784 88.2032 -90.508>}]
}
constraints [
 {bone1Name=dyn_driver_shirt_r, bone2Name=dyn_parent_shirt_r, swing1LimitDegrees=10, swing2LimitDegrees=10, twistLimitDegrees=5}
 {bone1Name=dyn_shirt_01_r, bone2Name=dyn_driver_shirt_r, swing1LimitDegrees=15, swing2LimitDegrees=17, twistLimitDegrees=5}
 {bone1Name=dyn_shirt_02_r, bone2Name=dyn_shirt_01_r, swing1LimitDegrees=18, swing2LimitDegrees=20, twistLimitDegrees=5}
 {bone1Name=dyn_parent_shirt_r, bone2Name=thigh_r, swing1LimitDegrees=45, swing2LimitDegrees=45, twistLimitDegrees=45}
 {bone1Name=dyn_shirt_03_r, bone2Name=dyn_shirt_02_r, swing1LimitDegrees=22, swing2LimitDegrees=25, twistLimitDegrees=5}
 {bone1Name=dyn_shirt_04_r, bone2Name=dyn_shirt_03_r, swing1LimitDegrees=26, swing2LimitDegrees=28, twistLimitDegrees=5}
 {
  bone1Name=calf_r
  bone2Name=thigh_r
  swing1LimitDegrees=45
  swing1Motion=Free
  swing2LimitDegrees=45
  swing2Motion=Free
  twistLimitDegrees=45
  twistMotion=Free
 }
]
