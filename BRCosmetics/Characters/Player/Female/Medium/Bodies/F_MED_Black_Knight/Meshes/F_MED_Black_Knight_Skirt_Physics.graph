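# Physics asset for the skirt of the medium female Black Knight player body, with capsule bodies on the skirt bones and both thighs.
# Four chains (left side, back-left, front-left, front-right) run from small parent capsules through driver bones to segment bones, linked by limited swing and tight twist constraints.
# Side and front parents attach to their thighs with free angular motion; the back-left parent has no thigh constraint.
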PhysicsAsset /BRCosmetics/Characters/Player/Female/Medium/Bodies/F_MED_Black_Knight/Meshes/F_MED_Black_Knight_Skirt_Physics
bodies {
 dyn_driver_shirt_bck_l [{shapeType=Capsule, length=3.5, radius=8.255, center=<0.75 0.0001 4.1829>, rotation=<rot 90 0 0>}]
 dyn_driver_shirt_frt_l [{shapeType=Capsule, length=4.5, radius=6.005, center=<2 0 3.873>, rotation=<rot 90 90 0>}]
 dyn_driver_shirt_frt_r [{shapeType=Capsule, length=4.25, radius=6.755, center=<1 0 0>, rotation=<rot 90 0 0>}]
 dyn_driver_shirt_l [{shapeType=Capsule, length=3.5, radius=8.255, center=<1 0 6.5974>, rotation=<rot 90 0 0>}]
 dyn_parent_shirt_bck_l [{shapeType=Capsule, length=0.505, radius=0.505, rotation=<rot 0 -90 0>}]
 dyn_parent_shirt_frt_l [{shapeType=Capsule, length=0.505, radius=0.505, rotation=<rot 0 -90 0>}]
 dyn_parent_shirt_frt_r [{shapeType=Capsule, length=0.505, radius=0.505, rotation=<rot 0 -90 0>}]
 dyn_parent_shirt_l [{shapeType=Capsule, length=0.505, radius=0.505, rotation=<rot 0 -90 0>}]
 dyn_shirt_01_l [{shapeType=Capsule, length=5.75, radius=6.255, center=<1.5 0.0001 4.717>, rotation=<rot 90 0 0>}]
 dyn_shirt_02_l [{shapeType=Capsule, length=4, radius=6.005, center=<2.0501 -0.0002 4.5571>, rotation=<rot 88.7046 0 0>}]
 dyn_shirt_03_l [{shapeType=Capsule, length=3.25, radius=5.255, center=<2.3923 0 3.207>, rotation=<rot 87.5179 0 0>}]
 dyn_shirt_04_l [{shapeType=Capsule, length=4, radius=3.255, center=<2.2702 -0.0003 1.2418>, rotation=<rot 86.0358 0 0>}]
 dyn_shirt_bck_01_l [{shapeType=Capsule, length=2.75, radius=6.505, center=<1.5 0 4.5757>, rotation=<rot 90 0 0>}]
 dyn_shirt_bck_02_l [{shapeType=Capsule, length=3, radius=5.255, center=<3 0 4.1636>, rotation=<rot 90 0 0>}]
 dyn_shirt_frt_01_l [{shapeType=Capsule, length=3.75, radius=5.005, center=<2 -0.0001 2.9159>, rotation=<rot 90 0 0>}]
 dyn_shirt_frt_01_r [{shapeType=Capsule, length=4.5, radius=6.255, center=<2 -0.0001 3.7514>, rotation=<rot 90 0 0>}]
 dyn_shirt_frt_02_l [{shapeType=Capsule, length=3, radius=4.005, center=<2 0.0001 1.097>, rotation=<rot 90 0 0>}]
 dyn_shirt_frt_02_r [{shapeType=Capsule, length=5, radius=4.505, center=<3.9838 0 2.1936>, rotation=<rot 87.2924 0 0>}]
 thigh_l [{shapeType=Capsule, length=27.3809, radius=9, center=<-18.6961 0.0084 -0.3178>, rotation=<rot 0 90.2866 -91.729>}]
 thigh_r [{shapeType=Capsule, length=27.6309, radius=6.9704, center=<22.7011 -0.1903 0.3735>, rotation=<rot 0.1296 93.7389 -92.1484>}]
}
constraints [
 {bone1Name=dyn_driver_shirt_bck_l, bone2Name=dyn_parent_shirt_bck_l, swing1LimitDegrees=15, swing2LimitDegrees=15, twistLimitDegrees=5}
 {bone1Name=dyn_shirt_bck_01_l, bone2Name=dyn_driver_shirt_bck_l, swing1LimitDegrees=15, swing2LimitDegrees=15, twistLimitDegrees=5}
 {bone1Name=dyn_shirt_01_l, bone2Name=dyn_driver_shirt_l, swing1LimitDegrees=14, swing2LimitDegrees=14, twistLimitDegrees=5}
 {bone1Name=dyn_shirt_frt_01_l, bone2Name=dyn_driver_shirt_frt_l, swing1LimitDegrees=20, swing2LimitDegrees=20, twistLimitDegrees=5}
 {bone1Name=dyn_driver_shirt_l, bone2Name=dyn_parent_shirt_l, swing1LimitDegrees=8, swing2LimitDegrees=8, twistLimitDegrees=5}
 {bone1Name=dyn_driver_shirt_frt_l, bone2Name=dyn_parent_shirt_frt_l, swing1LimitDegrees=20, swing2LimitDegrees=15, twistLimitDegrees=5}
 {bone1Name=dyn_driver_shirt_frt_r, bone2Name=dyn_parent_shirt_frt_r, swing1LimitDegrees=15, swing2LimitDegrees=30, twistLimitDegrees=5}
 {bone1Name=dyn_shirt_frt_01_r, bone2Name=dyn_driver_shirt_frt_r, swing1LimitDegrees=20, swing2LimitDegrees=45, twistLimitDegrees=5}
 {bone1Name=dyn_shirt_02_l, bone2Name=dyn_shirt_01_l, swing1LimitDegrees=12, swing2LimitDegrees=18, twistLimitDegrees=5}
 {bone1Name=dyn_shirt_frt_02_r, bone2Name=dyn_shirt_frt_01_r, swing1LimitDegrees=24, swing2LimitDegrees=45, twistLimitDegrees=5}
 {bone1Name=dyn_shirt_frt_02_l, bone2Name=dyn_shirt_frt_01_l, swing1LimitDegrees=20, swing2LimitDegrees=20, twistLimitDegrees=5}
 {bone1Name=dyn_shirt_bck_02_l, bone2Name=dyn_shirt_bck_01_l, swing1LimitDegrees=20, swing2LimitDegrees=20, twistLimitDegrees=5}
 {bone1Name=dyn_shirt_03_l, bone2Name=dyn_shirt_02_l, swing1LimitDegrees=15, swing2LimitDegrees=23, twistLimitDegrees=5}
 {bone1Name=dyn_shirt_04_l, bone2Name=dyn_shirt_03_l, swing1LimitDegrees=19, swing2LimitDegrees=30, twistLimitDegrees=5}
 {
  bone1Name=dyn_parent_shirt_l
  bone2Name=thigh_l
  swing1LimitDegrees=45
  swing1Motion=Free
  swing2LimitDegrees=45
  swing2Motion=Free
  twistLimitDegrees=45
  twistMotion=Free
 }
 {
  bone1Name=dyn_parent_shirt_frt_l
  bone2Name=thigh_l
  swing1LimitDegrees=45
  swing1Motion=Free
  swing2LimitDegrees=45
  swing2Motion=Free
  twistLimitDegrees=45
  twistMotion=Free
 }
 {
  bone1Name=dyn_parent_shirt_frt_r
  bone2Name=thigh_r
  swing1LimitDegrees=45
  swing1Motion=Free
  swing2LimitDegrees=45
  swing2Motion=Free
  twistLimitDegrees=45
  twistMotion=Free
 }
]
