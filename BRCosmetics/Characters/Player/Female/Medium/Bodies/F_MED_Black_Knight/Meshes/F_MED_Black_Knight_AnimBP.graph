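# Animation Blueprint for the female Black Knight body cosmetic, adding secondary motion on a pose copied from the attached parent mesh.
# Chained bone modifiers and weighted constraints correct shoulder pads, knees, straps, collar and knife, mostly driven by variables not computed here.
# Pelvis-space RigidBody nodes then simulate skirt, knife and bandana.
# Events enable backbling shoulder-pad adjustments and record crouch, emote and skydive states, which the AnimGraph never reads.
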AnimBlueprint /BRCosmetics/Characters/Player/Female/Medium/Bodies/F_MED_Black_Knight/Meshes/F_MED_Black_Knight_AnimBP
target_skeleton=&/F_MED_Black_Knight_Skeleton
graphs [
 AnimGraph {nodes [
   n1 = Get bIsWearingBackbling_b() -> (bIsWearingBackbling_b:bool)
   n2 = "Transform (Modify) Bone - Bone: dyn_driver_shldrpad_r"!(
    ComponentPose=Local To Component(LocalPose=Copy Pose From Mesh() {bUseAttachedParent=true})
    bAlphaBoolEnabled:bool=@n1
    Translation=Get ShldrPadRightPackTrans() -> (ShldrPadRightPackTrans:Vector)
    Rotation="0,0,0"
    Scale="1,1,0.92"
   ) -> (Pose) {
    alphaInputType=Bool
    scaleMode=BMM_Additive
    scaleSpace=BCS_BoneSpace
    translationMode=BMM_Additive
    translationSpace=BCS_BoneSpace
    alphaBoolBlend {blendOption=Linear}
    boneToModify {boneName=dyn_driver_shldrpad_r}
    scale {z=0.92}
   }
   n3 = "Transform (Modify) Bone - Bone: dyn_driver_shldrpad_l"!(ComponentPose=@n2, bAlphaBoolEnabled:bool=@n1, Translation="0,0,0", Rotation="0,0,0", Scale="1,1,0.92") -> (Pose) {
    alphaInputType=Bool
    scaleMode=BMM_Additive
    scaleSpace=BCS_BoneSpace
    alphaBoolBlend {blendOption=Linear}
    boneToModify {boneName=dyn_driver_shldrpad_l}
    scale {z=0.92}
   }
   n4 = "Constraint - dyn_knife_lookat (PPPPPPPP)"!(
    ComponentPose=@n3
    Alpha=1
    ConstraintWeights_0=0.2913
    ConstraintWeights_1:float=0.02
    ConstraintWeights_2:float=0.3619
    ConstraintWeights_3:float=0.0892
    ConstraintWeights_4:float=0.0204
    ConstraintWeights_5:float=0.1683
    ConstraintWeights_6:float=0.0349
    ConstraintWeights_7:float=0.0142
   ) -> (Pose) {
    alphaBoolBlend {blendOption=Linear}
    boneToModify {boneName=dyn_knife_lookat}
    constraintSetup [
     {targetBone {boneName=dyn_shirt_01_l}}
     {targetBone {boneName=dyn_shirt_bck_01_l}}
     {targetBone {boneName=thigh_twist_01_l}}
     {targetBone {boneName=dyn_driver_shirt_l}}
     {targetBone {boneName=dyn_shirt_bck_02_l}}
     {targetBone {boneName=dyn_shirt_02_l}}
     {targetBone {boneName=pelvis}}
     {targetBone {boneName=dyn_shirt_03_l}}
    ]
    constraintWeights [1,1,1,1,1,1,1,1]
   }
   n5 = "Constraint - dyn_parent_shldrpad_r (PP)"!(ComponentPose=@n4, Alpha=1, ConstraintWeights_0=0.1, ConstraintWeights_1:float=0.9) -> (Pose) {
    alphaBoolBlend {blendOption=Linear}
    boneToModify {boneName=dyn_parent_shldrpad_r}
    constraintSetup [{targetBone {boneName=clavicle_r}}, {targetBone {boneName=upperarm_twist_01_r}}]
    constraintWeights [1,1]
   }
   n6 = "Constraint - dyn_parent_shldrpad_l (PP)"!(ComponentPose=@n5, Alpha=1, ConstraintWeights_0=0.1, ConstraintWeights_1:float=0.9) -> (Pose) {
    alphaBoolBlend {blendOption=Linear}
    boneToModify {boneName=dyn_parent_shldrpad_l}
    constraintSetup [{targetBone {boneName=clavicle_l}}, {targetBone {boneName=upperarm_twist_01_l}}]
    constraintWeights [1,1]
   }
   n7 = "Transform (Modify) Bone - Bone: dyn_driver_shldrpad_l"!(
    ComponentPose=@n6
    Alpha:float=1
    Translation=Get ShldrPadLeftTrans() -> (ShldrPadLeftTrans:Vector)
    Rotation=Get ShldrPadLeftRotator() -> (ShldrPadLeftRotator:Rotator)
    Scale="1,1,1"
   ) -> (Pose) {
    rotationMode=BMM_Additive
    rotationSpace=BCS_BoneSpace
    translationMode=BMM_Additive
    translationSpace=BCS_BoneSpace
    alphaBoolBlend {blendOption=Linear}
    boneToModify {boneName=dyn_driver_shldrpad_l}
   }
   n8 = "Transform (Modify) Bone - Bone: dyn_driver_shldrpad_l"(
    ComponentPose=@n7
    Alpha=Get ShldrLeftAlpha() -> (ShldrLeftAlpha:float)
    Translation=Get ShldrPadLeftAimTrans() -> (ShldrPadLeftAimTrans:Vector)
    Rotation=Get ShldrPadLeftAimRotator() -> (ShldrPadLeftAimRotator:Rotator)
    Scale="0.85,0.85,0.8"
   ) {
    rotationMode=BMM_Additive
    rotationSpace=BCS_BoneSpace
    scaleMode=BMM_Additive
    scaleSpace=BCS_BoneSpace
    translationMode=BMM_Additive
    translationSpace=BCS_BoneSpace
    alphaBoolBlend {blendOption=Linear}
    alphaScaleBiasClamp {bInterpResult=true, interpSpeedDecreasing=25, interpSpeedIncreasing=25}
    boneToModify {boneName=dyn_driver_shldrpad_l}
    scale=<0.85 0.85 0.8>
   }
   n9 = "Transform (Modify) Bone - Bone: dyn_driver_shldrpad_l"(
    ComponentPose=@n8
    Alpha=1
    Translation=Get ShldrPadLeftRollTrans() -> (ShldrPadLeftRollTrans:Vector)
    Rotation=Get ShldrPadLeftRollRot() -> (ShldrPadLeftRollRot:Rotator)
    Scale="1,1,1"
   ) {
    rotationMode=BMM_Additive
    rotationSpace=BCS_BoneSpace
    scaleSpace=BCS_BoneSpace
    translationMode=BMM_Additive
    translationSpace=BCS_BoneSpace
    alphaBoolBlend {blendOption=Linear}
    alphaScaleBiasClamp {bInterpResult=true, interpSpeedDecreasing=25, interpSpeedIncreasing=25}
    boneToModify {boneName=dyn_driver_shldrpad_l}
   }
   n10 = "Transform (Modify) Bone - Bone: dyn_driver_shldrpad_l"(ComponentPose=@n9, Alpha=Get ShldrLeftRollAlpha() -> (ShldrLeftRollAlpha:float), Translation="10,-1,1", Rotation="12,-5,-80", Scale="1,1,1") {
    rotationMode=BMM_Additive
    rotationSpace=BCS_BoneSpace
    scaleMode=BMM_Additive
    scaleSpace=BCS_BoneSpace
    translationMode=BMM_Additive
    translationSpace=BCS_BoneSpace
    alphaBoolBlend {blendOption=Linear}
    alphaScaleBiasClamp {bInterpResult=true, interpSpeedDecreasing=50, interpSpeedIncreasing=50}
    boneToModify {boneName=dyn_driver_shldrpad_l}
    rotation=<rot 12 -5 -80>
    translation=<10 -1 1>
   }
   n11 = "Transform (Modify) Bone - Bone: dyn_driver_shldrpad_r"!(
    ComponentPose=@n10
    Alpha:float=1
    Translation=Get ShldrPadRightTrans() -> (ShldrPadRightTrans:Vector)
    Rotation=Get ShldrPadRightRotator() -> (ShldrPadRightRotator:Rotator)
    Scale="1,1,1"
   ) -> (Pose) {
    rotationMode=BMM_Additive
    rotationSpace=BCS_BoneSpace
    translationMode=BMM_Additive
    translationSpace=BCS_BoneSpace
    alphaBoolBlend {blendOption=Linear}
    boneToModify {boneName=dyn_driver_shldrpad_r}
   }
   n12 = "Transform (Modify) Bone - Bone: dyn_driver_shldrpad_r"(
    ComponentPose=@n11
    Alpha=Get ShldrRightAlpha() -> (ShldrRightAlpha:float)
    Translation=Get ShldrPadRightAimTrans() -> (ShldrPadRightAimTrans:Vector)
    Rotation=Get ShldrPadRightAimRotator() -> (ShldrPadRightAimRotator:Rotator)
    Scale="0.85,0.85,0.8"
   ) {
    rotationMode=BMM_Additive
    rotationSpace=BCS_BoneSpace
    scaleMode=BMM_Additive
    scaleSpace=BCS_BoneSpace
    translationMode=BMM_Additive
    translationSpace=BCS_BoneSpace
    alphaBoolBlend {blendOption=Linear}
    alphaScaleBiasClamp {bInterpResult=true, interpSpeedDecreasing=25, interpSpeedIncreasing=25}
    boneToModify {boneName=dyn_driver_shldrpad_r}
    scale=<0.85 0.85 0.8>
   }
   n13 = "Transform (Modify) Bone - Bone: dyn_driver_shldrpad_r"(
    ComponentPose=@n12
    Alpha=Get ShldrRightRollAlpha() -> (ShldrRightRollAlpha:float)
    Translation="-7,1,1"
    Rotation="-10,-5,-80"
    Scale="1,1,1"
   ) {
    rotationMode=BMM_Additive
    rotationSpace=BCS_BoneSpace
    scaleMode=BMM_Additive
    scaleSpace=BCS_BoneSpace
    translationMode=BMM_Additive
    translationSpace=BCS_BoneSpace
    alphaBoolBlend {blendOption=Linear}
    alphaScaleBiasClamp {bInterpResult=true, interpSpeedDecreasing=50, interpSpeedIncreasing=50}
    boneToModify {boneName=dyn_driver_shldrpad_r}
    rotation=<rot -10 -5 -80>
    translation=<-7 1 1>
   }
   n14 = "Transform (Modify) Bone - Bone: dyn_knee_l"(ComponentPose=@n13, Alpha=Get KneeBendLeftAlpha() -> (KneeBendLeftAlpha:float), Translation="1,-2,-0.000", Rotation="0,70,0", Scale="1,1,1") {
    rotationMode=BMM_Additive
    rotationSpace=BCS_BoneSpace
    scaleMode=BMM_Additive
    scaleSpace=BCS_BoneSpace
    translationMode=BMM_Additive
    translationSpace=BCS_BoneSpace
    alphaBoolBlend {blendOption=Linear}
    alphaScaleBiasClamp {bInterpResult=true, interpSpeedDecreasing=35, interpSpeedIncreasing=35}
    boneToModify {boneName=dyn_knee_l}
    rotation {yaw=70}
    translation=<1 -2>
   }
   n15 = "Transform (Modify) Bone - Bone: dyn_knee_r"(ComponentPose=@n14, Alpha=Get KneeBendRightAlpha() -> (KneeBendRightAlpha:float), Translation="-1,2,0.000", Rotation="0,70,0", Scale="1,1,1") {
    rotationMode=BMM_Additive
    rotationSpace=BCS_BoneSpace
    scaleMode=BMM_Additive
    scaleSpace=BCS_BoneSpace
    translationMode=BMM_Additive
    translationSpace=BCS_BoneSpace
    alphaBoolBlend {blendOption=Linear}
    alphaScaleBiasClamp {bInterpResult=true, interpSpeedDecreasing=35, interpSpeedIncreasing=35}
    boneToModify {boneName=dyn_knee_r}
    rotation {yaw=70}
    translation=<-1 2>
   }
   n16 = "Constraint - dyn_parent_knee_strap_out_r (PPPP)"!(
    ComponentPose=@n15
    Alpha=0.5
    ConstraintWeights_0=0.0107
    ConstraintWeights_1:float=0.9172
    ConstraintWeights_2:float=0.0147
    ConstraintWeights_3:float=0.0574
   ) -> (Pose) {
    alphaBoolBlend {blendOption=Linear}
    boneToModify {boneName=dyn_parent_knee_strap_out_r}
    constraintSetup [
     {targetBone {boneName=calf_twist_02_r}}
     {targetBone {boneName=dyn_knee_r}}
     {targetBone {boneName=thigh_twist_01_r}}
     {targetBone {boneName=calf_r}}
    ]
    constraintWeights [1,1,1,1]
   }
   n17 = "Constraint - dyn_parent_knee_strap_in_r (PPPP)"!(
    ComponentPose=@n16
    Alpha=0.5
    ConstraintWeights_0=0.0107
    ConstraintWeights_1:float=0.9169
    ConstraintWeights_2:float=0.015
    ConstraintWeights_3:float=0.0574
   ) -> (Pose) {
    alphaBoolBlend {blendOption=Linear}
    boneToModify {boneName=dyn_parent_knee_strap_in_r}
    constraintSetup [
     {targetBone {boneName=calf_twist_02_r}}
     {targetBone {boneName=dyn_knee_r}}
     {targetBone {boneName=thigh_twist_01_r}}
     {targetBone {boneName=calf_r}}
    ]
    constraintWeights [1,1,1,1]
   }
   n18 = "Constraint - dyn_parent_knee_strap_out_l (PPPP)"!(
    ComponentPose=@n17
    Alpha=0.5
    ConstraintWeights_0=0.0575
    ConstraintWeights_1:float=0.9171
    ConstraintWeights_2:float=0.0147
    ConstraintWeights_3:float=0.0107
   ) -> (Pose) {
    alphaBoolBlend {blendOption=Linear}
    boneToModify {boneName=dyn_parent_knee_strap_out_l}
    constraintSetup [
     {targetBone {boneName=calf_l}}
     {targetBone {boneName=dyn_knee_l}}
     {targetBone {boneName=thigh_twist_01_l}}
     {targetBone {boneName=calf_twist_02_l}}
    ]
    constraintWeights [1,1,1,1]
   }
   n19 = "Constraint - dyn_parent_knee_strap_in_l (PPPP)"!(
    ComponentPose=@n18
    Alpha=0.5
    ConstraintWeights_0=0.0574
    ConstraintWeights_1:float=0.9169
    ConstraintWeights_2:float=0.015
    ConstraintWeights_3:float=0.0107
   ) -> (Pose) {
    alphaBoolBlend {blendOption=Linear}
    boneToModify {boneName=dyn_parent_knee_strap_in_l}
    constraintSetup [
     {targetBone {boneName=calf_l}}
     {targetBone {boneName=dyn_knee_l}}
     {targetBone {boneName=thigh_twist_01_l}}
     {targetBone {boneName=calf_twist_02_l}}
    ]
    constraintWeights [1,1,1,1]
   }
   n20 = "Transform (Modify) Bone - Bone: dyn_driver_frnt_scarf"(
    ComponentPose=@n19
    Alpha=Get CollarFrontAlpha() -> (CollarFrontAlpha:float)
    Translation="-3,-0.000,2"
    Rotation="10,0,0.000"
    Scale="0.98,0.98,0.98"
   ) {
    rotationMode=BMM_Additive
    rotationSpace=BCS_BoneSpace
    scaleMode=BMM_Additive
    scaleSpace=BCS_BoneSpace
    translationMode=BMM_Additive
    translationSpace=BCS_BoneSpace
    alphaBoolBlend {blendOption=Linear}
    alphaScaleBiasClamp {bInterpResult=true, interpSpeedDecreasing=18, interpSpeedIncreasing=25}
    boneToModify {boneName=dyn_driver_frnt_scarf}
    rotation {pitch=10}
    scale=<0.98 0.98 0.98>
    translation {x=-3, z=2}
   }
   n21 = "Constraint - root_offset (RR)"!(ComponentPose=@n20, Alpha=1, ConstraintWeights_0=0.8, ConstraintWeights_1:float=0.2) -> (Pose) {
    alphaBoolBlend {blendOption=Linear}
    boneToModify {boneName=root_offset}
    constraintSetup [{transformType=Rotation, targetBone {boneName=root}}, {transformType=Rotation, targetBone {boneName=pelvis}}]
    constraintWeights [1,1]
   }
   n22 = RigidBody(ComponentPose=@n21) {
    bClampLinearTranslationLimitToRefPose=true
    bOverrideWorldGravity=true
    overridePhysicsAsset=&/F_MED_Black_Knight_Skirt_Physics
    simulationSpace=BaseBoneSpace
    alphaBoolBlend {blendOption=Linear}
    baseBoneRef {boneName=pelvis}
    componentAppliedLinearAccClamp=<1100 1100 1100>
    componentLinearAccScale=<0.4 0.4 0.5>
    componentLinearVelScale=<0.4 0.4 0.5>
    overrideWorldGravity {z=-850}
   }
   n23 = RigidBody(ComponentPose=@n22) {
    bClampLinearTranslationLimitToRefPose=true
    bOverrideWorldGravity=true
    overridePhysicsAsset=&/F_MED_Black_Knight_Right_Skirt_Physics
    simulationSpace=BaseBoneSpace
    alphaBoolBlend {blendOption=Linear}
    baseBoneRef {boneName=pelvis}
    componentAppliedLinearAccClamp=<1100 1100 1000>
    componentLinearAccScale=<0.1 0.1 0.1>
    componentLinearVelScale=<0.1 0.1 0.1>
    overrideWorldGravity {z=-980}
   }
   n24 = "Transform (Modify) Bone - Bone: dyn_parent_knife"(ComponentPose=@n23, Alpha=1, Translation="-1,-1.5,0", Rotation="0,0,-15", Scale="1,1,1") {
    rotationMode=BMM_Additive
    rotationSpace=BCS_BoneSpace
    scaleSpace=BCS_BoneSpace
    translationMode=BMM_Additive
    translationSpace=BCS_BoneSpace
    alphaBoolBlend {blendOption=Linear}
    alphaScaleBiasClamp {bInterpResult=true}
    boneToModify {boneName=dyn_parent_knife}
    rotation {roll=-15}
    translation=<-1 -1.5>
   }
   n25 = "Transform (Modify) Bone - Bone: dyn_parent_knife"(ComponentPose=@n24, Alpha=Get ThighRotLeft() -> (ThighRotLeft:float), Translation="-1,-0.75,1", Rotation="-15,0,-15", Scale="1,1,1") {
    rotationMode=BMM_Additive
    rotationSpace=BCS_BoneSpace
    scaleSpace=BCS_BoneSpace
    translationMode=BMM_Additive
    translationSpace=BCS_BoneSpace
    alphaBoolBlend {blendOption=Linear}
    alphaScaleBiasClamp {bInterpResult=true, interpSpeedDecreasing=25, interpSpeedIncreasing=25}
    boneToModify {boneName=dyn_parent_knife}
    rotation {pitch=-15, roll=-15}
    translation=<-1 -0.75 1>
   }
   n26 = "Transform (Modify) Bone - Bone: dyn_parent_knife"(ComponentPose=@n25, Alpha=Get ThighKnifeFix() -> (ThighKnifeFix:float), Translation="-0.7,-4.5,0", Rotation="-8,0,-5", Scale="1,1,1") {
    rotationMode=BMM_Additive
    rotationSpace=BCS_BoneSpace
    scaleSpace=BCS_BoneSpace
    translationMode=BMM_Additive
    translationSpace=BCS_BoneSpace
    alphaBoolBlend {blendOption=Linear}
    alphaScaleBiasClamp {bInterpResult=true, interpSpeedDecreasing=8, interpSpeedIncreasing=20}
    boneToModify {boneName=dyn_parent_knife}
    rotation {pitch=-8, roll=-5}
    translation=<-0.7 -4.5>
   }
   n27 = RigidBody(ComponentPose=@n26) {
    bClampLinearTranslationLimitToRefPose=true
    bOverrideWorldGravity=true
    overridePhysicsAsset=&/F_MED_Black_Knight_Knife_Physics
    simulationSpace=BaseBoneSpace
    alphaBoolBlend {blendOption=Linear}
    baseBoneRef {boneName=pelvis}
    componentAppliedLinearAccClamp=<2000 1100 1000>
    componentLinearAccScale=<0.3 0.4 0.3>
    componentLinearVelScale=<0.3 0.4 0.3>
    overrideWorldGravity {z=-980}
   }
   n28 = RigidBody(ComponentPose=@n27) {
    bClampLinearTranslationLimitToRefPose=true
    bOverrideWorldGravity=true
    overridePhysicsAsset=&/F_MED_Black_Knight_Bandana_Physics
    simulationSpace=BaseBoneSpace
    alphaBoolBlend {blendOption=Linear}
    baseBoneRef {boneName=pelvis}
    componentAppliedLinearAccClamp=<3000 3000 1000>
    componentLinearAccScale=<0.5 0.2 0.4>
    componentLinearVelScale=<0.5 0.2 0.4>
    overrideWorldGravity {z=-980}
   }
   Output Pose(Result=Component To Local(ComponentPose=@n28)) {layerGroup=None}
   Get ShldrLeftRifleAlpha() -> (ShldrLeftRifleAlpha:float)
  ]}
 EventGraph {nodes [
   Event OnBeginCrouching()
   > Set crouch_bool(crouch_bool:bool=true) -> (Output_Get:bool)
   > Set crouch_float(crouch_float:double=1) -> (Output_Get:double)
   Event OnEndCrouching()
   > Set crouch_bool(crouch_bool=false)
   > Set crouch_float(crouch_float=1)
   Event OnBeginPlayingEmote() {is=title K2Node_Event}
   > Set emote_active(emote_active:bool=true) -> (Output_Get:bool)
   Event BlueprintInitializeAnimation()
   > Set Skel_Mesh(Skel_Mesh=GetOwnerSkeletalMeshComponent())
   Event OnCharacterCustomizationLoaded()
   n1 = CheckForBackbling()
   > Set bIsWearingBackbling_b(bIsWearingBackbling_b:bool=OR Boolean(A=@n1, B=@n1.bIsBackblingLong)) -> (Output_Get:bool)
   Event OnEndPlayingEmote() {is=title K2Node_Event}
   > Set emote_active(emote_active=false)
   Event OnBeginSkydiving()
   > Set skydive_bool(skydive_bool:bool=true) -> (Output_Get:bool)
   Event OnEndSkydiving()
   > Set skydive_bool(skydive_bool=false)
   Reroute Node(InputPin:Wildcard) -> (OutputPin:Wildcard)
   Reroute Node()
  ]}
]
variables [
 bIsWearingBackbling_b {type=boolean}
 crouch_bool {type=boolean}
 crouch_float {type=number}
 emote_active {type=boolean}
 Skel_Mesh {type=SkeletalMeshComponent}
 skydive_bool {type=boolean}
]
